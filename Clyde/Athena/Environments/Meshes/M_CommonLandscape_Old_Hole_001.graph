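# M_CommonLandscape_Old_Hole_001 is a near-empty Default Lit material, named as a hole variant of an older common landscape material in the Athena environment set.
# Its graph holds one unnamed vector parameter with a grey default and no connection to any material output.
# No opacity mask or hole logic is present, so it reads as a stub.
# Automatic usage flagging is off and pixel depth offset uses legacy mode.
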Material /Clyde/Athena/Environments/Meshes/M_CommonLandscape_Old_Hole_001
graph {
 properties {bAutomaticallySetUsageInEditor=false, pixelDepthOffsetMode=PDOM_Legacy, usedShadingModels=Default Lit}
 nodes [VectorParameter() {defaultValue=<rgba 0.2159 0.2159 0.2159 0>}]
}
